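# Editor utility widget that switches PCG components between Normal and Preview editing modes from two buttons.
# SetActorSelection fills ActorArray with either all level actors (the default) or the current selection, depending on the AllActors flag, which a third button toggles.
# Each mode button loops those actors, gets the PCG component, and changes its live and serialized editing mode only if its display name contains "PCG_MeshGrass".
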EditorUtilityWidgetBlueprint /FigmentEnvironment/Tools/PCG/EUW_PCG_SwitchEditoMode
graphs [
 EventGraph {nodes [
   "On Clicked (EditorUtilityButton_85)"() {is=title K2Node_ComponentBoundEvent}
   > SetActorSelection()
   > n1 = For Each Loop(Array:Array of Actor Object References=Get ActorArray() -> (ActorArray:Array of Actor Object References)) -> (Array Element:Actor Object Reference)
   n2 = GetComponentByClass!(self:Actor Object Reference=@n1, ComponentClass:Actor Component Class Reference=/Script/PCG.PCGComponent) -> (ReturnValue:PCGComponent Object Reference)
   > Branch(Condition=IsPCGMeshGrass!(Object:Object Reference=@n2) -> (ReturnValue:bool))
   > SetEditingMode!(
    execute
    self:PCGComponent Object Reference=@n2
    InEditingMode:EPCGEditorDirtyMode Enum=Normal
    InSerializedEditingMode:EPCGEditorDirtyMode Enum=Normal
   ) -> (then)
   "On Clicked (EditorUtilityButton)"() {is=title K2Node_ComponentBoundEvent}
   > SetActorSelection()
   > n3 = For Each Loop(Array=Get ActorArray())
   n4 = GetComponentByClass(self=@n3, ComponentClass=&PCGComponent)
   > Branch(Condition=IsPCGMeshGrass(Object=@n4))
   > SetEditingMode(self=@n4, InEditingMode=Preview, InSerializedEditingMode=Preview)
   "On Clicked (EditorUtilityButton_SelectAll)"() {is=title K2Node_ComponentBoundEvent}
   > n5 = Set AllActors(AllActors:bool=NOT Boolean(A=Get AllActors() -> (AllActors:bool))) -> (Output_Get:bool)
   > PrintString(
    InString=Append(A="Select All Actors in Level: ", B="To String (Boolean)"(InBool=@n5))
    TextColor=(R=1,G=0,B=0.7398,A=1)
    Duration=5
    Key=SelectAlLState
   )
   Event PreConstruct()
   Event Construct()
   Event Tick()
  ]}
 IsPCGMeshGrass {nodes [
   n1 = IsPCGMeshGrass!() -> (then, Object:Object Reference) {is=title K2Node_FunctionEntry}
   > Return Node!(execute, ReturnValue:bool=Contains(SearchIn=GetDisplayName(Object=@n1), Substring=PCG_MeshGrass))
  ]}
 SetActorSelection {nodes [
   SetActorSelection() {is=title K2Node_FunctionEntry}
   > n1 = Branch(Condition=Get AllActors())
   > n2 = GetAllActorsOfClass!(execute, WorldContextObject:Object Reference, ActorClass:Actor Class Reference=&Actor) -> (then, OutActors:Array of Actor Object References)
   > Set ActorArray(ActorArray:Array of Actor Object References=@n2) -> (Output_Get:Array of Actor Object References)
   n3 = GetSelectedActors(execute=@n1.else)
   > Set ActorArray(ActorArray=@n3)
  ]}
]
variables [
 ActorArray {type="array<Actor>"}
 AllActors {type=boolean, default=true}
]
widgets {
 info {rootWidgetClass=&Overlay, widgetCount=8}
 widgets [
  Overlay_0 {widgetClassPath=&Overlay, widget=Overlay_0}
  VerticalBox_0 {widgetClassPath=&VerticalBox, parent=Overlay_0, widget=VerticalBox_0, slot {class=OverlaySlot}}
  EditorUtilityButton_85 {widgetClassPath=&EditorUtilityButton, parent=VerticalBox_0, widget=EditorUtilityButton_85, slot {class=VerticalBoxSlot}}
  TextBlock_27 {
   widgetClassPath=&TextBlock
   parent=EditorUtilityButton_85
   widget=TextBlock_27
   properties {text=Set Edit Mode to Normal}
   slot {class=ButtonSlot}
  }
  EditorUtilityButton {
   widgetClassPath=&EditorUtilityButton
   parent=VerticalBox_0
   widget=EditorUtilityButton
   slot {class=VerticalBoxSlot, properties {horizontalAlignment=HAlign_Left, verticalAlignment=VAlign_Top}}
  }
  TextBlock {
   widgetClassPath=&TextBlock
   parent=EditorUtilityButton
   widget=TextBlock
   properties {text=Set Edit Mode to Preview}
   slot {class=ButtonSlot}
  }
  EditorUtilityButton_SelectAll {
   widgetClassPath=&EditorUtilityButton
   parent=VerticalBox_0
   widget=EditorUtilityButton_SelectAll
   slot {class=VerticalBoxSlot, properties {horizontalAlignment=HAlign_Left, verticalAlignment=VAlign_Top}}
  }
  TextBlock_1 {
   widgetClassPath=&TextBlock
   parent=EditorUtilityButton_SelectAll
   widget=TextBlock_1
   properties {text=Select All Actors In level}
   slot {class=ButtonSlot}
  }
 ]
}
